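# Translucent UI material for the game phase message decoration in the battle royale HUD, reusing the compass frame texture flipped vertically.
# The texture's green channel acts as a distance field, which MF_UI_DFToMaskAndShadow turns into a thresholded shape mask and a soft glow.
# Animation scalars drive opacity and blend the adaptive Medium color toward StormPink outside the mask for the glow.
# Anim_Color currently has no visible effect because both blend inputs are the same color.
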Material /BattleRoyaleGameUI/HUD/Minimap/Assets/M_UI_GamePhaseMessage_Decor
graph {
 properties {bAutomaticallySetUsageInEditor=false, blendMode=BLEND_Translucent, materialDomain=MD_UI, usedShadingModels=Default Lit}
 nodes [
  n1 = MaterialFunctionCall() {materialFunction=&GetUserInterfaceUV}
  n2 = ComponentMask(None=@n1.Normalized UV) {g=true}
  n3 = TextureSample(UVs=AppendVector(A=ComponentMask(None=@n1.Normalized UV) {r=true}, B=OneMinus(None=@n2))) {samplerType=SAMPLERTYPE_LinearColor, texture=/BattleRoyaleGameUI/HUD/Compass/Assets/T_CompassFrame}
  n4 = MaterialFunctionCall!(DistanceField=@n3.G, Threshold=ScalarParameter() {defaultValue=0.425, parameterName=DecorationThreshold}, Smoothing) {materialFunction=/BattleRoyaleGameUI/Shared/MF_UI_DFToMaskAndShadow}
  n5 = LinearInterpolate(A=ScalarParameter() {defaultValue=0.25, parameterName=Opacity}, Alpha=ScalarParameter() {parameterName=Anim_Opacity})
  LinearInterpolate(A=Multiply(A=@n4.Mask, B=@n5), B=@n5, Alpha=@n4.Mask)
  LinearInterpolate(Alpha=ScalarParameter() {parameterName=Anim_Glow}) {constB=0.5}
  n6 = MaterialFunctionCall!(BufferSource) {materialFunction=/BattleRoyaleGameUI/Shared/MF_UI_AdaptiveColors}
  n7 = LinearInterpolate(A=@n6.Medium, B=@n6.Medium, Alpha=ScalarParameter() {parameterName=Anim_Color})
  n8 = CollectionParameter() {collection=/BattleRoyaleGameUI/Shared/MPC_BRGameUI, parameterName=StormPink}
  LinearInterpolate(
   A=@n7
   B=LinearInterpolate(A=ComponentMask(None=@n8) {b=true, g=true, r=true}, B=@n7, Alpha=@n4.Mask)
   Alpha=ScalarParameter() {parameterName=Anim_Glow}
  )
  Saturate(None=MaterialFunctionCall(DistanceField=@n3.G, Threshold=ScalarParameter() {parameterName=GlowThreshold}, Smoothing=ScalarParameter() {parameterName=GlowSmoothing}) {materialFunction=&MF_UI_DFToMaskAndShadow}.Shadow)
 ]
}
