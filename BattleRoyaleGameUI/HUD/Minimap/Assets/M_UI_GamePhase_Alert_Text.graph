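# Translucent UI material for the game phase alert text in the Battle Royale minimap HUD, providing an animatable color tint.
# It blends the shared White collection color toward the Bright output of MF_UI_AdaptiveColors by ColorMix, defaulting to an even mix.
# Anim_Color then interpolates from plain white to that mixed color, starting at white by default.
# A separate Opacity scalar parameter defaults to fully opaque.
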Material /BattleRoyaleGameUI/HUD/Minimap/Assets/M_UI_GamePhase_Alert_Text
graph {
 properties {bAutomaticallySetUsageInEditor=false, blendMode=BLEND_Translucent, materialDomain=MD_UI, usedShadingModels=Default Lit}
 nodes [
  n1 = ComponentMask(None=CollectionParameter() {collection=/BattleRoyaleGameUI/Shared/MPC_BRGameUI, parameterName=White}) {b=true, g=true, r=true}
  LinearInterpolate(
   A=@n1
   B=LinearInterpolate(
    A=@n1
    B=MaterialFunctionCall!(BufferSource) {materialFunction=/BattleRoyaleGameUI/Shared/MF_UI_AdaptiveColors}.Bright
    Alpha=ScalarParameter() {defaultValue=0.5, parameterName=ColorMix}
   )
   Alpha=ScalarParameter() {parameterName=Anim_Color}
  )
  ScalarParameter() {defaultValue=1, parameterName=Opacity}
 ]
}
