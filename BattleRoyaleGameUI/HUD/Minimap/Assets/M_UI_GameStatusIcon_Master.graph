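# Translucent UI master material for minimap HUD game status icons, drawing an icon over an angled backing shape.
# Box, BoxAngle, mirrored-UV and rotated top-edge masks form the backing, whose opacity has a low-end fallback.
# The icon texture supports distance-field antialiasing with derivative-based smoothing, single-channel or packed masks, plus scale and offset.
# Static switches choose shared UI colors or custom backing and icon colors, optionally pulsed between two by eased time sine.
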Material /BattleRoyaleGameUI/HUD/Minimap/Assets/M_UI_GameStatusIcon_Master
graph {
 properties {bAutomaticallySetUsageInEditor=false, blendMode=BLEND_Translucent, materialDomain=MD_UI, usedShadingModels=Default Lit}
 nodes [
  n1 = ScalarParameter() {defaultValue=0.5, parameterName=BackingOpacity}
  n2 = Constant() {r=0.3333}
  n3 = MaterialFunctionCall() {materialFunction=&GetUserInterfaceUV}
  n4 = MaterialFunctionCall(UVs=@n3.Normalized UV, Texture Scale=Constant() {r=0.95}) {materialFunction=&ScaleUVsByCenter}
  n5 = Constant() {r=2}
  n6 = ScalarParameter() {defaultValue=4, parameterName=Angle_2}
  n7 = ScalarParameter() {defaultValue=0.25, parameterName=AngleRotation}
  n8 = Constant() {r=0.95}
  n9 = AppendVector(A=@n8, B=Constant() {r=0.975})
  n10 = MaterialFunctionCall(
   UVs=MaterialFunctionCall!(
    UV=MaterialFunctionCall(UVs=@n3.Normalized UV, "Rotation Angle (0-1)"=@n7) {materialFunction=&CustomRotator}.Rotated Values
    Mirror Y=Constant() {r=1}
    Swizzle UV=Constant() {r=1}
   ) {materialFunction=/FortUILibrary/MaterialFunctions/Basics/MF_UI_Symmetry_2x1_New}.Symmetry
   Texture Scale=@n9
  ) {materialFunction=&ScaleUVsByCenter}
  n11 = Constant()
  n12 = ScalarParameter() {defaultValue=4, parameterName=Angle}
  n13 = LinearInterpolate(
   A=LinearInterpolate(
    A=MaterialFunctionCall!(UV=@n4.UVs, Corner=@n5, Angle=@n6, Flip, Falloff) {materialFunction=/FortUILibrary/MaterialFunctions/Shapes/MF_UI_BoxAngle}.Result
    B=MaterialFunctionCall(UV=@n10.UVs, Corner=@n11, Angle=@n12) {materialFunction=&MF_UI_BoxAngle}.Result
    Alpha=ScalarParameter() {parameterName=DoubleAngle}
   )
   B=NamedRerouteUsage()
   Alpha=ScalarParameter() {parameterName=UseAngleTop}
  )
  n14 = MaterialFunctionCall(Size=LinearInterpolate(A=@n8, B=@n9, Alpha=ScalarParameter() {parameterName=DoubleAngle})) {materialFunction=&MF_UI_Box}
  LinearInterpolate(
   A=Multiply(
    A=MaterialFunctionCall!(Default=@n1, LowEnd=@n2) {materialFunction=/BattleRoyaleGameUI/Shared/MF_UI_LowEnd_ValueSwitcher}.Result
    B=Min(A=@n13, B=@n14.Fill)
   )
   Alpha=NamedRerouteUsage()
  ) {constB=0.95}
  n15 = VectorParameter() {parameterName=BackingColor, defaultValue=<rgba 1 0 0 1>}
  n16 = MaterialFunctionCall("Value (0-1)"=MaterialFunctionCall(Speed=ScalarParameter() {defaultValue=1, parameterName=SineSpeed}) {materialFunction=&DebugTimeSine}.Result) {materialFunction=&MF_UI_EaseCurves}
  n17 = LinearInterpolate(A=@n15.RGB, B=VectorParameter() {parameterName=BackingColor_2, defaultValue=<rgba 1 0 0 1>}.RGB, Alpha=@n16.EaseOutCubic)
  n18 = VectorParameter() {parameterName=IconColor, defaultValue=<rgba 0 1 0 1>}
  n19 = LinearInterpolate(A=@n18.RGB, B=VectorParameter() {parameterName=IconColor_2, defaultValue=<rgba 0 1 0 1>}.RGB, Alpha=@n16.EaseOutCubic)
  n20 = CollectionParameter() {collection=/BattleRoyaleGameUI/Shared/MPC_BRGameUI, parameterName=White}
  LinearInterpolate(
   A=StaticSwitchParameter(
    True=StaticSwitchParameter(True=@n17, False=@n15.RGB) {parameterName=SineColor}
    False=MaterialFunctionCall!(BufferSource) {materialFunction=/BattleRoyaleGameUI/Shared/MF_UI_AdaptiveColors}.Dark
   ) {parameterName=CustomBackingColor}
   B=StaticSwitchParameter(True=StaticSwitchParameter(True=@n19, False=@n18.RGB) {parameterName=SineColor}, False=ComponentMask(None=@n20) {b=true, g=true, r=true}) {parameterName=CustomIconColor}
   Alpha=NamedRerouteUsage()
  )
  n21 = ScalarParameter() {defaultValue=0.2, parameterName=IconThreshold}
  n22 = Add(
   A=MaterialFunctionCall(UVs=TextureCoordinate(), Texture Scale=ScalarParameter() {defaultValue=1, parameterName=IconScale}) {materialFunction=&ScaleUVsByCenter}.UVs
   B=AppendVector(A=ScalarParameter() {parameterName=IconTranslate_X}, B=ScalarParameter() {parameterName=IconTranslate_Y})
  )
  n23 = Abs(None=DDX(Value=@n22))
  n24 = Abs(None=DDY(Value=@n22))
  n25 = Add(A=ScalarParameter() {parameterName=IconSmoothing}, B=Multiply(A=Max(A=@n23, B=@n24), B=ScalarParameter() {parameterName=SmoothScale}))
  n26 = TextureSampleParameter2D(UVs=@n22) {parameterName=Texture, samplerSource=SSM_Clamp_WorldGroupSettings, texture=/Game/Athena/Items/Accolades/Materials/Textures/Time128px}
  n27 = StaticComponentMaskParameter(None=@n26.RGBA) {defaultG=true, parameterName=Channel}
  n28 = LinearInterpolate(A=@n26.B, Alpha=LinearInterpolate(A=@n26.G, Alpha=@n26.G))
  NamedRerouteDeclaration(None=StaticSwitchParameter(
    True=MaterialFunctionCall(DistanceField=@n27, Threshold=@n21, Smoothing=@n25) {materialFunction=&MF_UI_AntiAliasing}.Outer
    False=StaticSwitchParameter(True=@n28, False=@n27) {parameterName=MultiChannelPacked}
   ) {defaultValue=true, parameterName=DFTexture}) {name=IconMask, nodeColor=<rgba 0 0 0 1>}
  n29 = MaterialFunctionCall() {materialFunction=&GetUserInterfaceUV}
  n30 = MaterialFunctionCall!(Degrees=Constant() {r=0.75}) {materialFunction=/Engine/Functions/Engine_MaterialFunctions01/Math/DegreesToRadians}
  n31 = ComponentMask(None=Add(
    A=MaterialFunctionCall(UVs=@n29.Normalized UV, "Rotation Angle (0-1)"=@n30.Radians) {materialFunction=&CustomRotator}.Rotated Values
    B=Constant2Vector() {g=0.44}
   )) {g=true}
  NamedRerouteDeclaration(None=MaterialFunctionCall(DistanceField=MaterialFunctionCall!("Texture mask (0-1 range)"=@n31) {materialFunction=/FortUILibrary/MaterialFunctions/Shapes/Modifiers/MF_UI_SDF_RemapTextureRange}.SDF) {materialFunction=&MF_UI_AntiAliasing}.Outer) {name=TopMask, nodeColor=<rgba 0.6406 1 0 1>}
  ScalarParameter() {defaultValue=0.95, parameterName=DoubleVersionTextureScale}
 ]
}
