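# Translucent UI material that draws the rounded game status backing in the Battle Royale minimap HUD.
# Offset circle masks and an anti-aliased, angled linear gradient are combined with min and lerp into a curved fill region and an edge mask.
# A rotated-UV smoothstep fades the fill, which blends toward a separate opacity along the edge.
# Color comes from the dark output of the shared adaptive colors function.
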Material /BattleRoyaleGameUI/HUD/Minimap/Assets/M_UI_GameStatus_RoundedBacking
graph {
 properties {bAutomaticallySetUsageInEditor=false, blendMode=BLEND_Translucent, materialDomain=MD_UI, usedShadingModels=Default Lit}
 nodes [
  n1 = Min(A=NamedRerouteUsage(), B=NamedRerouteUsage())
  n2 = MaterialFunctionCall() {materialFunction=&GetUserInterfaceUV}
  n3 = ScalarParameter() {defaultValue=-0.115, parameterName=BackingRotationAngle}
  n4 = OneMinus(None=ComponentMask(None=MaterialFunctionCall(UVs=@n2.Normalized UV, "Rotation Angle (0-1)"=@n3) {materialFunction=&CustomRotator}.Rotated Values) {g=true})
  n5 = SmoothStep(Min=ScalarParameter() {defaultValue=0.5, parameterName=Min}, Value=@n4)
  n6 = ScalarParameter() {defaultValue=0.75, parameterName=Opacity}
  n7 = ScalarParameter() {defaultValue=0.105, parameterName=AnglePosition}
  n8 = Constant() {r=0.25}
  n9 = MaterialFunctionCall!(Degrees=ScalarParameter() {defaultValue=-1, parameterName=Angle}) {materialFunction=/Engine/Functions/Engine_MaterialFunctions01/Math/DegreesToRadians}
  n10 = MaterialFunctionCall(
   DistanceField=MaterialFunctionCall(GradientStart=OneMinus(None=@n7), GradientEnd=Constant() {r=1}, RotationAngle=Add(A=@n8, B=@n9.Radians)) {materialFunction=&MF_UI_LinearGradient}.Result
   Threshold=Constant() {r=0.5}
  ) {materialFunction=&MF_UI_AntiAliasing}
  n11 = MaterialFunctionCall() {materialFunction=&MF_UI_UVScaling}
  n12 = AppendVector(
   A=ScalarParameter() {defaultValue=-1.5, parameterName=CircleEdge_OffsetX}
   B=ScalarParameter() {defaultValue=0.75, parameterName=CircleEdge_OffsetY}
  )
  n13 = ScalarParameter() {defaultValue=2.65, parameterName=CircleMask_Size}
  n14 = OneMinus(None=MaterialFunctionCall(UVs=Add(A=@n11.Scaled to Fit, B=@n12), Size=Add(A=@n13) {constB=0.025}) {materialFunction=&MF_UI_Circle}.Fill)
  n15 = Add(
   A=MaterialFunctionCall() {materialFunction=&MF_UI_UVScaling}.Scaled to Fit
   B=AppendVector(
    A=ScalarParameter() {defaultValue=-1.6, parameterName=CircleMask_OffsetX}
    B=ScalarParameter() {defaultValue=0.6, parameterName=CircleMask_OffsetY}
   )
  )
  n16 = ScalarParameter() {defaultValue=2.65, parameterName=CircleMask_Size}
  n17 = Min(
   A=LinearInterpolate(A=Saturate(None=@n10.Outer), Alpha=OneMinus(None=@n14))
   B=OneMinus(None=MaterialFunctionCall(UVs=@n15, Size=@n16) {materialFunction=&MF_UI_Circle}.Fill)
  )
  LinearInterpolate(A=Multiply(A=ScalarParameter() {defaultValue=0.25, parameterName=FillOpacity}, B=Multiply(A=@n1, B=@n5)), B=@n6, Alpha=@n17)
  n18 = MaterialFunctionCall!(BufferSource) {materialFunction=/BattleRoyaleGameUI/Shared/MF_UI_AdaptiveColors}
  LinearInterpolate(A=@n18.Dark, B=@n18.Dark, Alpha=@n17)
  Multiply(A=@n17, B=@n6)
  NamedRerouteDeclaration(None=@n14) {name=Mask, nodeColor=<rgba 0.9219 1 0 1>}
  NamedRerouteDeclaration(None=OneMinus(None=@n10.Outer)) {name=Mask2, nodeColor=<rgba 0.7578 0 1 1>}
  Saturate(None=MaterialFunctionCall("Value (0-1)"=@n4) {materialFunction=&MF_UI_EaseCurves}.EaseInCubic)
 ]
}
