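# Translucent UI material that draws the minimap key icon from a selectable channel of a distance field texture.
# MF_UI_AntiAliasing thresholds the field twice, yielding a tight inner shape and a wider surrounding one.
# The inner mask blends shared MPC_BRGameUI colors from Charcoal to White, forming a white icon with a dark edge.
# A weighted sum of both masks forms a combined alpha, with the outer shape weighted lower.
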Material /BattleRoyaleGameUI/HUD/Minimap/Assets/M_UI_MinimapKeyIcon
graph {
 properties {bAutomaticallySetUsageInEditor=false, blendMode=BLEND_Translucent, materialDomain=MD_UI, usedShadingModels=Default Lit}
 nodes [
  n1 = TextureObjectParameter() {
   parameterName=Texture Input
   texture=&/T_UI_Icons_HUDT_UI_MiniMap_Key_128pxT_UI_GameStatusIcons_DF_128px
  }
  n2 = StaticComponentMaskParameter(None=TextureSample(Tex=@n1) {texture=&T_UI_Icons_HUDT_UI_MiniMap_Key_128pxT_UI_GameStatusIcons_DF_128px}.RGBA) {defaultR=true, parameterName=Channel}
  n3 = MaterialFunctionCall(DistanceField=@n2, Threshold=Constant() {r=0.22}) {materialFunction=&MF_UI_AntiAliasing}
  n4 = Constant() {r=0.02}
  Add(
   A=Multiply(A=@n3.Outer) {constB=0.95}
   B=Multiply(A=MaterialFunctionCall(DistanceField=@n2, Threshold=@n4) {materialFunction=&MF_UI_AntiAliasing}.Outer) {constB=0.7}
  )
  LinearInterpolate(
   A=CollectionParameter() {collection=/BattleRoyaleGameUI/Shared/MPC_BRGameUI, parameterName=Charcoal}
   B=CollectionParameter() {collection=&MPC_BRGameUI, parameterName=White}
   Alpha=@n3.Outer
  )
 ]
}
