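# Translucent UI font material, named for point-of-interest location names on the minimap, with animatable opacity and color swipe.
# Anim_Opacity blends from zero to the Opacity parameter.
# Rotated UVs, offset by Anim_Swipe between stop and start values, feed a sine shaped by cubic ease-in into a color blend.
# Both blend inputs are the same Bright color from MF_UI_AdaptiveColors, so the swipe leaves the color unchanged.
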Material /BattleRoyaleGameUI/HUD/Minimap/Assets/M_UI_POI_LocationName_Font
graph {
 properties {bAutomaticallySetUsageInEditor=false, blendMode=BLEND_Translucent, materialDomain=MD_UI, usedShadingModels=Default Lit}
 nodes [
  LinearInterpolate(B=ScalarParameter() {defaultValue=1, parameterName=Opacity}, Alpha=ScalarParameter() {defaultValue=1, parameterName=Anim_Opacity})
  n1 = MaterialFunctionCall!(BufferSource) {materialFunction=/BattleRoyaleGameUI/Shared/MF_UI_AdaptiveColors}
  n2 = Add(
   A=MaterialFunctionCall(UVs=TextureCoordinate(), "Rotation Angle (0-1)"=ScalarParameter() {defaultValue=0.2, parameterName=SwipeRotationAngle}) {materialFunction=&CustomRotator}.Rotated Values
   B=LinearInterpolate(
    A=ScalarParameter() {defaultValue=-1, parameterName=Swipe_Stop}
    B=ScalarParameter() {defaultValue=1, parameterName=Swipe_Start}
    Alpha=ScalarParameter() {parameterName=Anim_Swipe}
   )
  )
  n3 = Saturate(None=ComponentMask(None=Sine(None=@n2) {period=2}) {g=true})
  LinearInterpolate(
   A=@n1.Bright
   B=@n1.Bright
   Alpha=Saturate(None=MaterialFunctionCall("Value (0-1)"=@n3) {materialFunction=&MF_UI_EaseCurves}.EaseInCubic)
  )
 ]
}
