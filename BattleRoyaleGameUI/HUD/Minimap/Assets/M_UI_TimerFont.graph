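# UI font material for the minimap HUD timer, drawing two-color glowing text from the font's signed distance field.
# Eased distance values feed an anti-aliasing function, whose outer mask blends secondary to primary color and shapes opacity.
# The eased distance also scales color, which sAnim_Color fades to black and vertex alpha modulates.
# Anim_Opacity fades both branches, while AdditiveMixRatio lowers alpha under AlphaComposite blending for a partly additive look.
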Material /BattleRoyaleGameUI/HUD/Minimap/Assets/M_UI_TimerFont
graph {
 properties {bAutomaticallySetUsageInEditor=false, blendMode=BLEND_AlphaComposite, materialDomain=MD_UI, usedShadingModels=Default Lit}
 nodes [
  n1 = MaterialFunctionCall("Value (0-1)"=FontSignedDistance().Smooth Signed Distance) {materialFunction=&MF_UI_EaseCurves}
  n2 = MaterialFunctionCall(
   DistanceField=MaterialFunctionCall("Value (0-1)"=@n1.EaseInQuint) {materialFunction=&MF_UI_EaseCurves}.EaseInCubic
   Threshold=ScalarParameter() {defaultValue=0.5, parameterName=Threshold}
   Smoothing=ScalarParameter() {parameterName=Smoothing}
  ) {materialFunction=&MF_UI_AntiAliasing}
  n3 = ScalarParameter() {defaultValue=0.5, parameterName=AdditiveMixRatio}
  Multiply(A=@n2.Outer, B=LinearInterpolate(B=OneMinus(None=@n3), Alpha=ScalarParameter() {parameterName=Anim_Opacity}))
  n4 = LinearInterpolate(
   A=LinearInterpolate(
    A=VectorParameter() {parameterName=SecondaryColor, defaultValue=<rgba 0.1667 0 1 1>}.RGB
    B=VectorParameter() {parameterName=PrimaryColor, defaultValue=<rgba 0.9647 0.0409 0.7991 1>}.RGB
    Alpha=@n2.Outer
   )
   Alpha=ScalarParameter() {parameterName=sAnim_Color}
  ) {constB=0}
  Multiply(A=VertexColor().A, B=LinearInterpolate(B=Multiply(A=@n4, B=@n1.EaseInQuint), Alpha=ScalarParameter() {parameterName=Anim_Opacity}))
 ]
}
